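# Sound cue for the end of the Icicle Larva transformation, classed as a Fort gadget/item sound.
# A Branch on the IsLocal bool parameter plays the transform-end wave only when true; the false and unset inputs are empty, so it is silent otherwise.
# The cue boosts volume and sends to the Icicle Larva flight sidechain volume audio bus, apparently to drive ducking.
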SoundCue /IcicleLarvaGameplay/Assets/Audio/Cues/Transformation/IP_SC_IcicyleLarva_End_Transformation
root=Branch_0
nodes [
 {id=Branch_0, children [WavePlayer_2,null,null], properties {boolParameterName=IsLocal}}
 {id=WavePlayer_2, properties {soundWaveAssetPtr=/IcicleLarvaGameplay/Assets/Audio/Waves/IP_IcicleLarva_Transform_End}}
]
properties {
 maxDistance=2097152
 soundClassObject=&Fort_Gadgets_Items
 volumeMultiplier=1.5
 busSends [{audioBus=/IcicleLarvaGameplay/Assets/Audio/Modulation/AB_IcicleLarva_Flight_SideChain_Vol, minSendLevel=0, sendLevel=1, soundSourceBus=null}]
}
