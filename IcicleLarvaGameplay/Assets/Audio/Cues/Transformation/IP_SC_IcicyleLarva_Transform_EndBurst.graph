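# Sound cue for the burst that ends the Icicle Larva transformation, layering close and distant versions under separate attenuation settings.
# The close layer mixes the burst wave with a louder, delayed random pick of three booms.
# The distant layer plays one of two random distant waves, equally delayed and pitched slightly down.
# The cue also sends to a flight sidechain volume audio bus.
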SoundCue /IcicleLarvaGameplay/Assets/Audio/Cues/Transformation/IP_SC_IcicyleLarva_Transform_EndBurst
root=Mixer_1
nodes [
 {
  id=Attenuation_0
  children [Mixer_0]
  properties {attenuationSettings=/IcicleLarvaGameplay/Assets/Audio/Settings/ATT_IcicleLarva_Transformation_Close}
 }
 {
  id=Attenuation_1
  children [Modulator_1]
  properties {attenuationSettings=/IcicleLarvaGameplay/Assets/Audio/Settings/ATT_IcicleLarva_Transformation_Distant}
 }
 {id=Delay_0, children [Random_0], properties {delayMax=1, delayMin=1}}
 {id=Delay_1, children [Random_1], properties {delayMax=1, delayMin=1}}
 {id=Mixer_0, children [Modulator_0,Delay_0], properties {inputVolume [0.7,1.8]}}
 {id=Mixer_1, children [Attenuation_0, Attenuation_1], properties {inputVolume [1,1]}}
 {id=Modulator_0, children [WavePlayer_0], properties {pitchMax=1, pitchMin=1, volumeMax=1, volumeMin=1}}
 {id=Modulator_1, children [Delay_1], properties {pitchMax=0.9, pitchMin=0.9}}
 {id=Random_0, children [WavePlayer_1,WavePlayer_2,WavePlayer_3], properties {weights [1,1,1]}}
 {id=Random_1, children [WavePlayer_7,WavePlayer_8], properties {weights [1,1]}}
 {id=WavePlayer_0, properties {soundWaveAssetPtr=/IcicleLarvaGameplay/Assets/Audio/Waves/IP_IcicleLarva_Transform_EndBurst}}
 {id=WavePlayer_1, properties {soundWaveAssetPtr=/IcicleLarvaGameplay/Assets/Audio/Waves/IP_IcicleLarva_Transform_End_Boom_01}}
 {id=WavePlayer_2, properties {soundWaveAssetPtr=/IcicleLarvaGameplay/Assets/Audio/Waves/IP_IcicleLarva_Transform_End_Boom_02}}
 {id=WavePlayer_3, properties {soundWaveAssetPtr=/IcicleLarvaGameplay/Assets/Audio/Waves/IP_IcicleLarva_Transform_End_Boom_03}}
 {id=WavePlayer_7, properties {soundWaveAssetPtr=/IcicleLarvaGameplay/Assets/Audio/Waves/IP_IcicleLarva_Transform_EndBurst_Distant_01}}
 {id=WavePlayer_8, properties {soundWaveAssetPtr=/IcicleLarvaGameplay/Assets/Audio/Waves/IP_IcicleLarva_Transform_EndBurst_Distant_02}}
]
properties {
 maxDistance=28000
 soundClassObject=&Fort_Gadgets_Items
 volumeMultiplier=1
 busSends [{audioBus=/IcicleLarvaGameplay/Assets/Audio/Modulation/AB_IcicleLarva_Flight_SideChain_Vol, minSendLevel=0, sendLevel=1, soundSourceBus=null}]
}
