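# Sound cue for the take-off moment of the Icicle Larva transformation, in the Fortnite gadgets/items sound class.
# A mixer layers two wave players at equal volume, a take-off wave and its distant variant.
# Each layer has its own attenuation asset, close and distant respectively, so the sound changes with listener distance.
# The cue also sends at full level to an Icicle Larva flight sidechain volume audio bus.
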SoundCue /IcicleLarvaGameplay/Assets/Audio/Cues/Transformation/IP_SC_IcicyleLarva_Transform_TakeOff
root=Mixer_0
nodes [
 {
  id=Attenuation_0
  children [WavePlayer_0]
  properties {attenuationSettings=/IcicleLarvaGameplay/Assets/Audio/Settings/ATT_IcicleLarva_Transformation_Close}
 }
 {
  id=Attenuation_1
  children [WavePlayer_1]
  properties {attenuationSettings=/IcicleLarvaGameplay/Assets/Audio/Settings/ATT_IcicleLarva_Transformation_Distant}
 }
 {id=Mixer_0, children [Attenuation_0, Attenuation_1], properties {inputVolume [1,1]}}
 {id=WavePlayer_0, properties {soundWaveAssetPtr=/IcicleLarvaGameplay/Assets/Audio/Waves/IP_IcicleLarva_Transform_TakeOff}}
 {id=WavePlayer_1, properties {soundWaveAssetPtr=/IcicleLarvaGameplay/Assets/Audio/Waves/IP_IcicleLarva_Transform_TakeOff_Distant}}
]
properties {
 maxDistance=28000
 soundClassObject=&Fort_Gadgets_Items
 volumeMultiplier=1
 busSends [{audioBus=/IcicleLarvaGameplay/Assets/Audio/Modulation/AB_IcicleLarva_Flight_SideChain_Vol, minSendLevel=0, sendLevel=1, soundSourceBus=null}]
}
